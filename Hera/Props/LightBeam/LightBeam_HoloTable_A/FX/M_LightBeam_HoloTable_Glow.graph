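# Unlit, two-sided translucent material for the glowing light cone of the holo table beam, usable on Niagara mesh particles.
# Color blends dark to cone tint by a steep UV-V gradient, then passes through MF_FX_EmissiveAdjustments.
# Opacity layers a cone gradient, panning noise-distorted rays and cloud-texture smoke, softened by inverted Fresnel, depth fade and a squared-sine V falloff.
# A hot V-edge band is added before particle alpha scales the result.
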Material /Hera/Props/LightBeam/LightBeam_HoloTable_A/FX/M_LightBeam_HoloTable_Glow
graph {
 properties {
  bUsedWithNanite=true
  bUsedWithNiagaraMeshParticles=true
  blendMode=BLEND_Translucent
  pixelDepthOffsetMode=PDOM_Legacy
  shadingModel=MSM_Unlit
  twoSided=true
  usedShadingModels=Unlit
 }
 nodes [
  Add()
  n1 = Power(Base=ComponentMask(None=TextureCoordinate()) {g=true}) {constExponent=30}
  MaterialFunctionCall(Emissive=LinearInterpolate(
    A=VectorParameter() {group=Color, parameterName=DarkColor, defaultValue=<rgba 0.2 0.3333 1 1>}.RGB
    B=VectorParameter() {group=Color, parameterName=Cone Color, defaultValue=<rgba 0.6 1 3 1>}.RGB
    Alpha=@n1
   )) {materialFunction=&MF_FX_EmissiveAdjustments}
  n2 = ComponentMask(None=TextureCoordinate()) {g=true}
  n3 = ScalarParameter() {defaultValue=6, group=Opacity, parameterName=Cone Exponent}
  n4 = Panner(Coordinate=TextureCoordinate() {vTiling=0}, Speed=VectorParameter() {parameterName=PannerSpeed, defaultValue=<rgba 0.05 0.1 0 0>}.RGB) {speedX=0.05, speedY=0.1}
  n5 = Add(
   A=Multiply(A=TextureSample(UVs=@n4) {texture=&T_NoisePack_01}.R) {constB=0.05}
   B=TextureCoordinate() {uTiling=8, vTiling=0}
  )
  n6 = VectorParameter() {parameterName=PannerSpeed, defaultValue=<rgba 0.05 0.1 0 0>}
  n7 = LinearInterpolate(
   A=ScalarParameter() {defaultValue=-1, group=Opacity, parameterName=Opacity Remap}
   Alpha=TextureSample(UVs=Panner(Coordinate=@n5, Speed=@n6.RGB) {speedX=0.05, speedY=0.1}) {texture=&T_NoisePack_01}.R
  ) {constA=-1.5}
  n8 = Power(Base=@n2, Exp=ScalarParameter() {defaultValue=5, group=Opacity, parameterName=Ray Exponent}) {constExponent=5}
  n9 = Multiply(A=@n2, B=@n2)
  n10 = Panner() {speedX=0.1, speedY=0.2}
  n11 = Add(A=LinearInterpolate(Alpha=TextureSample(UVs=@n10) {texture=/Game/Packages/Fortnite_UniqueMaterials_1/PostProcess/T_TilingClouds_02}.R) {constA=-0.3, constB=0.1})
  n12 = Add(
   A=Add(A=Power(Base=@n2, Exp=@n3) {constExponent=10}, B=Multiply(A=@n7, B=@n8))
   B=Multiply(A=Multiply(A=@n9, B=@n11), B=ScalarParameter() {defaultValue=0.2, group=Opacity, parameterName=Smoke Strength})
  )
  n13 = Fresnel(ExponentIn=ScalarParameter() {defaultValue=1, group=Opacity, parameterName=Fresnel Exponent}) {baseReflectFraction=0, exponent=3}
  n14 = ScalarParameter() {defaultValue=5, group=Opacity, parameterName=DepthFadeDistance}
  n15 = TextureCoordinate()
  n16 = Sine(None=Multiply(A=ComponentMask(None=@n15) {g=true}) {constB=0.5})
  n17 = Multiply(A=Multiply(A=OneMinus(None=@n13), B=DepthFade(FadeDistance=@n14) {fadeDistanceDefault=5}), B=Multiply(A=@n16, B=@n16))
  Multiply(A=Add(A=Multiply(A=@n1) {constB=5}, B=Multiply(A=@n12, B=@n17)), B=ParticleColor().A)
  Multiply()
 ]
}
